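# Niagara system for the Doodle vehicle boost cosmetic, drawing a boost cone, trail and small sprites at filtered skeletal-mesh sockets.
# Active emitters are layered cone sprites split by spawn group and render tag, a multi-renderer ribbon trail blown backward, and a cone-velocity sprite spray.
# User.GameModeID switches some sizes, offsets and lifetimes.
# Alternate cone and trail emitters are disabled, as are all painted-variant override modules.
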NiagaraSystem /VehicleCosmetics/Boosts/Boost_Doodle/FX/NS_BoostDoodle
dependencies {
 usedDataInterfaces [
  &NiagaraDataInterfaceColorCurve
  &NiagaraDataInterfaceCurve
  &NiagaraDataInterfaceArrayFloat
  &NiagaraDataInterfaceSkeletalMesh
 ]
 usedDynamicInputs []
 usedModules [
  &self:SystemSpawnScript.SystemScriptSource.SystemScriptGraph.NiagaraNodeAssignment_0.SetVariables_#1
  &SystemState
 ]
}
emitters {
 NE_BoostCone {
  emitterUpdateScript [
   &EmitterState(Inactive Response=::NewEnumerator1, Life Cycle Mode=::NewEnumerator1, Loop Duration=1, UseLoopDelay=false)
   &SpawnBurst_Instantaneous(Spawn Count=2, Spawn Group=4, Use Spawn Probability=false, UseLoopCountLimit=false) {module=SpawnBurst_Instantaneous004}
   &SpawnBurst_Instantaneous(Spawn Count=2, Spawn Group=3, Use Spawn Probability=false, UseLoopCountLimit=false) {module=SpawnBurst_Instantaneous003}
   &SpawnBurst_Instantaneous(Spawn Count=2, Spawn Group=2, Use Spawn Probability=false, UseLoopCountLimit=false) {module=SpawnBurst_Instantaneous002}
   &SpawnBurst_Instantaneous(Spawn Count=2, Spawn Group=1, Use Spawn Probability=false, UseLoopCountLimit=false) {module=SpawnBurst_Instantaneous001}
   &SpawnBurst_Instantaneous(Spawn Count=2, Use Spawn Probability=false, UseLoopCountLimit=false)
   &VC_FilteredSocketCount()
   &SpawnRate(SpawnRate=50, Use Spawn Probability=false) {enabled=false}
  ]
  particleSpawnScript [
   &VC_PaintedVariantsOverride() {enabled=false}
   &InitializeParticle(
    Color=&MaskLinearColorBySpawnGroup(
     Masked Linear Color=<rgba 1 1 1 1>
     Passthrough Linear Color=&MaskLinearColorBySpawnGroup(
      Masked Linear Color=<rgba 0 0.0013 0.0625 100>
      Passthrough Linear Color=&MaskLinearColorBySpawnGroup(
       Masked Linear Color=<rgba 1 1 1 50>
       Passthrough Linear Color=&MaskLinearColorBySpawnGroup(
        Masked Linear Color=<rgba 0 0.0013 0.0625 100>
        Passthrough Linear Color=&MaskLinearColorBySpawnGroup(Masked Linear Color=<rgba 0 0.0013 0.0625 100>, Passthrough Linear Color=<rgba 0.4179 0 0 10>)
       )
       Spawn Group Mask=1
      )
      Spawn Group Mask=2
     )
     Spawn Group Mask=3
    )
    Sprite Size Mode=::NewEnumerator3
    Uniform Sprite Size=&MaskFloatBySpawnGroup(
     Masked Float=33.9821
     Passthrough Float=&MaskFloatBySpawnGroup(
      Masked Float=25
      Passthrough Float=&MaskFloatBySpawnGroup(
       Masked Float=&MergeHelper_FloatSwitch(0=53, 1=53, 2=57, 3=?, Index=$User.GameModeID, Num=::NewEnumerator1)
       Passthrough Float=&MaskFloatBySpawnGroup(
        Masked Float=&MergeHelper_FloatSwitch(0=52, 1=52, 2=58, 3=?, Index=$User.GameModeID, Num=::NewEnumerator1)
        Passthrough Float=28.7828
       )
       Spawn Group Mask=1
      )
      Spawn Group Mask=2
     )
     Spawn Group Mask=3
    )
    UsePositionOffset=false
    Write Material Random=true
   )
   &PartitionParticles(Number Of Partitions=$Emitter.PortCount)
   &SkeletalMeshLocation(
    Bone / Socket Index=$Particles.PartitionParticles.Partition
    Bone / Socket Sampling Mode=::NewEnumerator5
    MaskSpawnGroup=false
    Mesh Sampling Type=::NewEnumerator6
    Sampled Position Offset=<-57.8812 1 -0.1677>
    Skeletal Mesh=$System.SkeletalMesh
   )
   &AddVelocity(Velocity=<-1500 0 0>) {enabled=false}
   &CameraOffset(Camera Offset Amount=&MaskFloatBySpawnGroup(
     Masked Float=51
     Passthrough Float=&MaskFloatBySpawnGroup(
      Masked Float=52
      Passthrough Float=&MaskFloatBySpawnGroup(
       Masked Float=5
       Passthrough Float=&MaskFloatBySpawnGroup(Masked Float=-20, Passthrough Float=50)
       Spawn Group Mask=1
      )
      Spawn Group Mask=2
     )
     Spawn Group Mask=3
    ))
   /SetVariables(Particles.RenderTag:NiagaraInt32=&SpawnGroup()) {isSetParameters=true, module=SetVariables_#2}
  ]
  particleUpdateScript [
   &ParticleState()
   &SolveForcesAndVelocity(Clamp Velocity=false, Limit Acceleration=false)
  ]
  rendererClasses [&NiagaraSpriteRendererProperties]
  renderers [
   {rendererClass=&NiagaraSpriteRendererProperties}
   {rendererClass=&NiagaraSpriteRendererProperties, rendererIndex=1}
   {rendererClass=&NiagaraSpriteRendererProperties, rendererIndex=2}
   {rendererClass=&NiagaraSpriteRendererProperties, rendererIndex=3}
   {rendererClass=&NiagaraSpriteRendererProperties, rendererIndex=4}
  ]
 }
 NE_BoostCone001 {
  bEnabled=false
  emitterUpdateScript [
   &EmitterState(Inactive Response=::NewEnumerator1, Life Cycle Mode=::NewEnumerator1, Loop Duration=1, UseLoopDelay=false)
   &SpawnBurst_Instantaneous(Spawn Count=2, Spawn Group=4, Use Spawn Probability=false, UseLoopCountLimit=false) {module=SpawnBurst_Instantaneous004}
   &SpawnBurst_Instantaneous(Spawn Count=2, Spawn Group=3, Use Spawn Probability=false, UseLoopCountLimit=false) {module=SpawnBurst_Instantaneous003}
   &SpawnBurst_Instantaneous(Spawn Count=2, Spawn Group=2, Use Spawn Probability=false, UseLoopCountLimit=false) {module=SpawnBurst_Instantaneous002}
   &SpawnBurst_Instantaneous(Spawn Count=2, Spawn Group=1, Use Spawn Probability=false, UseLoopCountLimit=false) {module=SpawnBurst_Instantaneous001}
   &SpawnBurst_Instantaneous(Spawn Count=2, Use Spawn Probability=false, UseLoopCountLimit=false)
   &VC_FilteredSocketCount()
   &SpawnRate(SpawnRate=50, Use Spawn Probability=false) {enabled=false}
  ]
  particleSpawnScript [
   &VC_PaintedVariantsOverride() {enabled=false}
   &InitializeParticle(
    Color=&MaskLinearColorBySpawnGroup(
     Masked Linear Color=<rgba 1 1 1 500>
     Passthrough Linear Color=&MaskLinearColorBySpawnGroup(
      Masked Linear Color=<rgba 0 0.0013 0.0625 100>
      Passthrough Linear Color=&MaskLinearColorBySpawnGroup(
       Masked Linear Color=<rgba 1 1 1 500>
       Passthrough Linear Color=&MaskLinearColorBySpawnGroup(
        Masked Linear Color=<rgba 0 0.0013 0.0625 106>
        Passthrough Linear Color=&MaskLinearColorBySpawnGroup(Masked Linear Color=<rgba 0 0.0013 0.0625 100>, Passthrough Linear Color=<rgba 0.4179 0 0 10>)
       )
       Spawn Group Mask=1
      )
      Spawn Group Mask=2
     )
     Spawn Group Mask=3
    )
    Sprite Size Mode=::NewEnumerator3
    Uniform Sprite Size=&MaskFloatBySpawnGroup(
     Masked Float=26.2665
     Passthrough Float=&MaskFloatBySpawnGroup(
      Masked Float=34
      Passthrough Float=&MaskFloatBySpawnGroup(
       Masked Float=44
       Passthrough Float=&MaskFloatBySpawnGroup(Masked Float=55, Passthrough Float=45)
       Spawn Group Mask=1
      )
      Spawn Group Mask=2
     )
     Spawn Group Mask=3
    )
    UsePositionOffset=false
    Write Material Random=true
   )
   &PartitionParticles(Number Of Partitions=$Emitter.PortCount)
   &SkeletalMeshLocation(
    Bone / Socket Index=$Particles.PartitionParticles.Partition
    Bone / Socket Sampling Mode=::NewEnumerator5
    MaskSpawnGroup=false
    Mesh Sampling Type=::NewEnumerator6
    Sampled Position Offset=<-42.5 0 -0.1677>
    Skeletal Mesh=$System.SkeletalMesh
   )
   &AddVelocity(Velocity=<-1500 0 0>) {enabled=false}
   &CameraOffset(Camera Offset Amount=&MaskFloatBySpawnGroup(
     Masked Float=45
     Passthrough Float=&MaskFloatBySpawnGroup(
      Masked Float=50
      Passthrough Float=&MaskFloatBySpawnGroup(
       Masked Float=0
       Passthrough Float=&MaskFloatBySpawnGroup(Masked Float=-25, Passthrough Float=40)
       Spawn Group Mask=1
      )
      Spawn Group Mask=2
     )
     Spawn Group Mask=3
    )) {enabled=false}
   /SetVariables(Particles.RenderTag:NiagaraInt32=&SpawnGroup()) {isSetParameters=true, module=SetVariables_#2}
  ]
  particleUpdateScript [
   &ParticleState()
   &SolveForcesAndVelocity(Clamp Velocity=false, Limit Acceleration=false)
  ]
  rendererClasses [&NiagaraSpriteRendererProperties]
  renderers [
   {rendererClass=&NiagaraSpriteRendererProperties}
   {rendererClass=&NiagaraSpriteRendererProperties, rendererIndex=1}
   {rendererClass=&NiagaraSpriteRendererProperties, rendererIndex=2}
   {rendererClass=&NiagaraSpriteRendererProperties, rendererIndex=3}
   {rendererClass=&NiagaraSpriteRendererProperties, rendererIndex=4}
  ]
 }
 NE_BoostCone002 {
  bEnabled=false
  emitterUpdateScript [
   &EmitterState(Inactive Response=::NewEnumerator1, Life Cycle Mode=::NewEnumerator1, Loop Duration=1, UseLoopDelay=false)
   &SpawnBurst_Instantaneous(Spawn Count=2, Spawn Group=4, Use Spawn Probability=false, UseLoopCountLimit=false) {module=SpawnBurst_Instantaneous004}
   &SpawnBurst_Instantaneous(Spawn Count=2, Spawn Group=3, Use Spawn Probability=false, UseLoopCountLimit=false) {module=SpawnBurst_Instantaneous003}
   &SpawnBurst_Instantaneous(Spawn Count=2, Spawn Group=2, Use Spawn Probability=false, UseLoopCountLimit=false) {module=SpawnBurst_Instantaneous002}
   &SpawnBurst_Instantaneous(Spawn Count=2, Spawn Group=1, Use Spawn Probability=false, UseLoopCountLimit=false) {module=SpawnBurst_Instantaneous001}
   &SpawnBurst_Instantaneous(Spawn Count=2, Use Spawn Probability=false, UseLoopCountLimit=false)
   &VC_FilteredSocketCount()
   &SpawnRate(SpawnRate=50, Use Spawn Probability=false) {enabled=false}
  ]
  particleSpawnScript [
   &VC_PaintedVariantsOverride() {enabled=false}
   &InitializeParticle(
    Color=&MaskLinearColorBySpawnGroup(
     Masked Linear Color=<rgba 1 1 1 50>
     Passthrough Linear Color=&MaskLinearColorBySpawnGroup(
      Masked Linear Color=<rgba 0 0.0013 0.0625 100>
      Passthrough Linear Color=&MaskLinearColorBySpawnGroup(
       Masked Linear Color=<rgba 1 1 1 100>
       Passthrough Linear Color=&MaskLinearColorBySpawnGroup(
        Masked Linear Color=<rgba 0 0.0013 0.0625 100>
        Passthrough Linear Color=&MaskLinearColorBySpawnGroup(Masked Linear Color=<rgba 0 0.0013 0.0625 100>, Passthrough Linear Color=<rgba 0.4179 0 0 10>)
       )
       Spawn Group Mask=1
      )
      Spawn Group Mask=2
     )
     Spawn Group Mask=3
    )
    Sprite Size Mode=::NewEnumerator3
    Uniform Sprite Size=&MaskFloatBySpawnGroup(
     Masked Float=26.2665
     Passthrough Float=&MaskFloatBySpawnGroup(
      Masked Float=34
      Passthrough Float=&MaskFloatBySpawnGroup(
       Masked Float=44
       Passthrough Float=&MaskFloatBySpawnGroup(Masked Float=55, Passthrough Float=45)
       Spawn Group Mask=1
      )
      Spawn Group Mask=2
     )
     Spawn Group Mask=3
    )
    UsePositionOffset=false
    Write Material Random=true
   )
   &PartitionParticles(Number Of Partitions=$Emitter.PortCount)
   &SkeletalMeshLocation(
    Bone / Socket Index=$Particles.PartitionParticles.Partition
    Bone / Socket Sampling Mode=::NewEnumerator5
    MaskSpawnGroup=false
    Mesh Sampling Type=::NewEnumerator6
    Sampled Position Offset=<-42.5 0 -0.1677>
    Skeletal Mesh=$System.SkeletalMesh
   )
   &AddVelocity(Velocity=<-1500 0 0>) {enabled=false}
   &CameraOffset(Camera Offset Amount=&MaskFloatBySpawnGroup(
     Masked Float=45
     Passthrough Float=&MaskFloatBySpawnGroup(
      Masked Float=50
      Passthrough Float=&MaskFloatBySpawnGroup(
       Masked Float=0
       Passthrough Float=&MaskFloatBySpawnGroup(Masked Float=-25, Passthrough Float=40)
       Spawn Group Mask=1
      )
      Spawn Group Mask=2
     )
     Spawn Group Mask=3
    )) {enabled=false}
   /SetVariables(Particles.RenderTag:NiagaraInt32=&SpawnGroup()) {isSetParameters=true, module=SetVariables_#2}
  ]
  particleUpdateScript [
   &ParticleState()
   &SolveForcesAndVelocity(Clamp Velocity=false, Limit Acceleration=false)
  ]
  rendererClasses [&NiagaraSpriteRendererProperties]
  renderers [
   {rendererClass=&NiagaraSpriteRendererProperties}
   {rendererClass=&NiagaraSpriteRendererProperties, rendererIndex=1}
   {rendererClass=&NiagaraSpriteRendererProperties, rendererIndex=2}
   {rendererClass=&NiagaraSpriteRendererProperties, rendererIndex=3}
   {rendererClass=&NiagaraSpriteRendererProperties, rendererIndex=4}
  ]
 }
 NE_BoostCone003 {
  bEnabled=false
  emitterUpdateScript [
   &EmitterState(Inactive Response=::NewEnumerator1, Life Cycle Mode=::NewEnumerator1, Loop Duration=1, UseLoopDelay=false)
   &SpawnBurst_Instantaneous(Spawn Count=2, Spawn Group=4, Use Spawn Probability=false, UseLoopCountLimit=false) {module=SpawnBurst_Instantaneous004}
   &SpawnBurst_Instantaneous(Spawn Count=2, Spawn Group=3, Use Spawn Probability=false, UseLoopCountLimit=false) {module=SpawnBurst_Instantaneous003}
   &SpawnBurst_Instantaneous(Spawn Count=2, Spawn Group=2, Use Spawn Probability=false, UseLoopCountLimit=false) {module=SpawnBurst_Instantaneous002}
   &SpawnBurst_Instantaneous(Spawn Count=2, Spawn Group=1, Use Spawn Probability=false, UseLoopCountLimit=false) {module=SpawnBurst_Instantaneous001}
   &SpawnBurst_Instantaneous(Spawn Count=2, Use Spawn Probability=false, UseLoopCountLimit=false)
   &VC_FilteredSocketCount()
   &SpawnRate(SpawnRate=50, Use Spawn Probability=false) {enabled=false}
  ]
  particleSpawnScript [
   &VC_PaintedVariantsOverride() {enabled=false}
   &InitializeParticle(
    Color=&MaskLinearColorBySpawnGroup(
     Masked Linear Color=<rgba 1 1 1 50>
     Passthrough Linear Color=&MaskLinearColorBySpawnGroup(
      Masked Linear Color=<rgba 0 0.0013 0.0625 100>
      Passthrough Linear Color=&MaskLinearColorBySpawnGroup(
       Masked Linear Color=<rgba 1 1 1 100>
       Passthrough Linear Color=&MaskLinearColorBySpawnGroup(
        Masked Linear Color=<rgba 0 0.0013 0.0625 100>
        Passthrough Linear Color=&MaskLinearColorBySpawnGroup(Masked Linear Color=<rgba 0 0.0013 0.0625 100>, Passthrough Linear Color=<rgba 0.4179 0 0 10>)
       )
       Spawn Group Mask=1
      )
      Spawn Group Mask=2
     )
     Spawn Group Mask=3
    )
    Sprite Size Mode=::NewEnumerator3
    Uniform Sprite Size=&MaskFloatBySpawnGroup(
     Masked Float=20
     Passthrough Float=&MaskFloatBySpawnGroup(
      Masked Float=25
      Passthrough Float=&MaskFloatBySpawnGroup(
       Masked Float=53.1395
       Passthrough Float=&MaskFloatBySpawnGroup(Masked Float=55.7975, Passthrough Float=28)
       Spawn Group Mask=1
      )
      Spawn Group Mask=2
     )
     Spawn Group Mask=3
    )
    UsePositionOffset=false
    Write Material Random=true
   )
   &PartitionParticles(Number Of Partitions=$Emitter.PortCount)
   &SkeletalMeshLocation(
    Bone / Socket Index=$Particles.PartitionParticles.Partition
    Bone / Socket Sampling Mode=::NewEnumerator5
    MaskSpawnGroup=false
    Mesh Sampling Type=::NewEnumerator6
    Sampled Position Offset=<-57.8812 0 -0.1677>
    Skeletal Mesh=$System.SkeletalMesh
   )
   &AddVelocity(Velocity=<-1500 0 0>) {enabled=false}
   &CameraOffset(Camera Offset Amount=&MaskFloatBySpawnGroup(
     Masked Float=51
     Passthrough Float=&MaskFloatBySpawnGroup(
      Masked Float=52
      Passthrough Float=&MaskFloatBySpawnGroup(
       Masked Float=5
       Passthrough Float=&MaskFloatBySpawnGroup(Masked Float=-20, Passthrough Float=50)
       Spawn Group Mask=1
      )
      Spawn Group Mask=2
     )
     Spawn Group Mask=3
    ))
   /SetVariables(Particles.RenderTag:NiagaraInt32=&SpawnGroup()) {isSetParameters=true, module=SetVariables_#2}
  ]
  particleUpdateScript [
   &ParticleState()
   &SolveForcesAndVelocity(Clamp Velocity=false, Limit Acceleration=false)
  ]
  rendererClasses [&NiagaraSpriteRendererProperties]
  renderers [
   {rendererClass=&NiagaraSpriteRendererProperties}
   {rendererClass=&NiagaraSpriteRendererProperties, rendererIndex=1}
   {rendererClass=&NiagaraSpriteRendererProperties, rendererIndex=2}
   {rendererClass=&NiagaraSpriteRendererProperties, rendererIndex=3}
   {rendererClass=&NiagaraSpriteRendererProperties, rendererIndex=4}
  ]
 }
 NE_DoodleMainTrail {
  bEnabled=false
  emitterUpdateScript [
   &EmitterState()
   &VC_FilteredSocketCount()
   &VC_RateQualityHandler(High Spawn Rate=60, Low Spawn Rate=30, Use Custom Rate?=true)
   &VC_SpawnRibbon_v1(Number of Ribbons=$Emitter.PortCount, SpawnRate=60)
  ]
  particleSpawnScript [
   &VC_PaintedVariantsOverride() {enabled=false}
   &VC_InitializeRibbon_v1(
    Color=<rgba 1 1 1 10000>
    Enable Skeletal Position=true
    Lifetime=&MergeHelper_FloatSwitch(0=0.6, 1=0.6, 2=0.3, 3=?, Index=$User.GameModeID, Num=::NewEnumerator1)
    Number of Bones Per Boneflow group=2
    Offset Amount:NiagaraFloat=&MergeHelper_FloatSwitch(0=-35, 1=-35, 2=-45, 3=?, Index=$User.GameModeID, Num=::NewEnumerator1)
    Offset Axis:ENiagaraOrientationAxis=::XAxis
   )
   &AddVelocity(Velocity=&MergeHelper_VectorSwitch(0=<-1300 0 0>, 1=<-1300 0 0>, 2=<-1300 0 0>, 3=?, Index=$User.GameModeID, Num=::NewEnumerator1)) {module=AddVelocity001}
  ]
  particleUpdateScript [
   &ParticleState()
   &InheritSourceMovement(
    Applied Position Scale=&Multiply_VectorByFloat(
     Vector=<1 1 1>
     Float=&FloatFromCurve(FloatCurve=DataInterface() {propertyValues {Curve {defaultValue=0, keys [<at 0 1 RCIM_Cubic>,<at 1 0.8 RCIM_Cubic>]}}})
    )
    ApplyForce=false
    ApplyPosition=true
    ApplyPositionFalloff=false
   )
   &SolveForcesAndVelocity(Clamp Velocity=false, Limit Acceleration=false)
   /SetVariables(
    Particles.ColorTag02:LinearColor=&MultiplyLinearColor(
     Linear Color=<rgba 0 0.0012 0.063 500>
     Scale Factor=&MakeVector4FromVectorAndFloat("Vector (RGB)"=<1 1 1>, "Float (Alpha)"=1)
    )
    Particles.WidthTag02=66.9773
    Particles.WidthTagMain=41
    Particles.float:NiagaraFloat=&SelectFloatFromArray(Direct Array Index=&SpawnGroup(), Float Selection Array=DataInterface() {propertyValues {FloatData [-42,0]}})
    Particles.float001:NiagaraFloat=&SelectFloatFromArray(Direct Array Index=&SpawnGroup(), Float Selection Array=DataInterface() {propertyValues {FloatData [0,42]}})
   ) {isSetParameters=true, module=SetVariables_#3}
   &ScaleColor(
    Scale Alpha=&FloatFromCurve(FloatCurve=DataInterface() {propertyValues {Curve {defaultValue=1, keys [{tangentMode=RCTM_User, time=0.2, value=0}, {tangentMode=RCTM_User, time=1, value=1}]}}})
    ScaleA=true
    ScaleRGB=true
   ) {enabled=false}
   &VC_UpdateRibbon_v1(Enable Skeletal Position=true)
   &Color(Color=&ColorFromCurve(ColorCurve=DataInterface() {propertyValues {
       AlphaCurve {keys [<at 0.1059 1>,<at 0.4665 50>,<at 1 1 RCIM_Cubic>]}
       BlueCurve {keys [<at 0.1131 0.4115 RCIM_Cubic>,<at 0.4564 1>,<at 0.9971 0.063 RCIM_Cubic>]}
       GreenCurve {keys [<at 0.1131 0.092 RCIM_Cubic>,<at 0.4564 1>,<at 0.9971 0.0012 RCIM_Cubic>]}
       RedCurve {keys [<at 0.1131 0.0857 RCIM_Cubic>,<at 0.4564 1>,<at 0.9971 0 RCIM_Cubic>]}
      }})) {enabled=false}
   &ScaleRibbonWidth(Ribbon Width Scale=&SelectFloatFromArray(Direct Array Index=&SpawnGroup(), Float Selection Array=DataInterface() {propertyValues {FloatData [1,-1]}}))
  ]
  rendererClasses [&NiagaraRibbonRendererProperties]
  renderers [
   {rendererClass=&NiagaraRibbonRendererProperties}
   {rendererClass=&NiagaraRibbonRendererProperties, rendererIndex=1}
   {rendererClass=&NiagaraRibbonRendererProperties, rendererIndex=2}
   {rendererClass=&NiagaraRibbonRendererProperties, rendererIndex=3}
  ]
 }
 NE_DoodleMainTrail001 {
  emitterUpdateScript [
   &EmitterState()
   &VC_FilteredSocketCount()
   &VC_RateQualityHandler(High Spawn Rate=60, Low Spawn Rate=30, Use Custom Rate?=true)
   &VC_SpawnRibbon_v1(Number of Ribbons=$Emitter.PortCount, SpawnRate=60)
  ]
  particleSpawnScript [
   &VC_PaintedVariantsOverride() {enabled=false}
   &VC_InitializeRibbon_v1(
    Enable Skeletal Position=true
    Lifetime=&MergeHelper_FloatSwitch(0=0.6, 1=0.6, 2=0.3, 3=?, Index=$User.GameModeID, Num=::NewEnumerator1)
    Number of Bones Per Boneflow group=2
    Offset Amount:NiagaraFloat=&MergeHelper_FloatSwitch(0=-45, 1=-45, 2=-55, 3=?, Index=$User.GameModeID, Num=::NewEnumerator1)
    Offset Axis:ENiagaraOrientationAxis=::XAxis
   )
   &AddVelocity(Velocity=&MergeHelper_VectorSwitch(0=<-1300 0 0>, 1=<-1300 0 0>, 2=<-1300 0 0>, 3=?, Index=$User.GameModeID, Num=::NewEnumerator1)) {module=AddVelocity001}
   /SetVariables(
    Particles.ColorTag02:LinearColor=&MultiplyLinearColor(
     Linear Color=<rgba 0 0.0013 0.0625 50>
     Scale Factor=&MakeVector4FromVectorAndFloat("Vector (RGB)"=<1 1 1>, "Float (Alpha)"=1)
    )
    Particles.WidthTag02:NiagaraFloat=&MergeHelper_FloatSwitch(0=47, 1=47, 2=47, 3=?, Index=0, Num=::NewEnumerator1)
    Particles.WidthTagMain=43
    Particles.float:NiagaraFloat=&SelectFloatFromArray(Direct Array Index=&SpawnGroup(), Float Selection Array=DataInterface() {propertyValues {FloatData [-40,0]}})
    Particles.float001:NiagaraFloat=&SelectFloatFromArray(Direct Array Index=&SpawnGroup(), Float Selection Array=DataInterface() {propertyValues {FloatData [0,40]}})
   ) {isSetParameters=true, module=SetVariables_#3}
  ]
  particleUpdateScript [
   &ParticleState()
   &InheritSourceMovement(
    Applied Position Scale=&Multiply_VectorByFloat(
     Vector=<1 1 1>
     Float=&FloatFromCurve(FloatCurve=DataInterface() {propertyValues {Curve {defaultValue=0, keys [<at 0 1 RCIM_Cubic>,<at 1 0.8 RCIM_Cubic>]}}})
    )
    ApplyForce=false
    ApplyPosition=true
    ApplyPositionFalloff=false
   )
   &SolveForcesAndVelocity(Clamp Velocity=false, Limit Acceleration=false)
   &ScaleColor(
    Scale Alpha=&FloatFromCurve(FloatCurve=DataInterface() {propertyValues {Curve {defaultValue=1, keys [{tangentMode=RCTM_User, time=0, value=1}, {tangentMode=RCTM_User, time=1, value=1}]}}})
    ScaleA=true
    ScaleRGB=true
   )
   &VC_UpdateRibbon_v1(Enable Skeletal Position=true)
   &Color(Color=&ColorFromCurve(ColorCurve=DataInterface() {propertyValues {
       AlphaCurve {keys [<at 0 1 RCIM_Cubic>, {arriveTangent=-1.3117, interpMode=RCIM_Cubic, time=0.441, value=10}, <at 0.8463 0>]}
       BlueCurve {keys [<at 0 1>,<at 0.7131 0.1146>]}
       GreenCurve {keys [<at 0 1>,<at 0.7131 0.0307>]}
       RedCurve {keys [<at 0 1>,<at 0.7131 0.0221>]}
      }}))
   /SetVariables(
    Particles.BGB:LinearColor=&MultiplyLinearColor(
     Linear Color=<rgba 0 0.0052 0.25 60>
     Scale Factor=&MakeVector4FromVectorAndFloat(
      "Vector (RGB)"=<1 1 1>
      "Float (Alpha)"=&FloatFromCurve(FloatCurve=DataInterface() {propertyValues {Curve {defaultValue=0.878, keys [{tangentMode=RCTM_User, time=0, value=1}, {tangentMode=RCTM_User, time=1, value=1}]}}})
     )
    )
    Particles.BubblesColor:LinearColor=&MultiplyLinearColor(
     Linear Color=<rgba 0 0.0013 0.0625 50>
     Scale Factor=&MakeVector4FromVectorAndFloat(
      "Vector (RGB)"=<1 1 1>
      "Float (Alpha)"=&FloatFromCurve(FloatCurve=DataInterface() {propertyValues {Curve {keys [<at 0.5 1>,<at 1 0>]}}})
     )
    )
   ) {isSetParameters=true, module=SetVariables_#4}
  ]
  rendererClasses [&NiagaraRibbonRendererProperties]
  renderers [
   {rendererClass=&NiagaraRibbonRendererProperties}
   {rendererClass=&NiagaraRibbonRendererProperties, rendererIndex=1}
   {rendererClass=&NiagaraRibbonRendererProperties, rendererIndex=2}
   {rendererClass=&NiagaraRibbonRendererProperties, rendererIndex=3}
   {rendererClass=&NiagaraRibbonRendererProperties, rendererIndex=4}
  ]
 }
 NE_DoodleMainTrail002 {
  bEnabled=false
  emitterUpdateScript [
   &EmitterState()
   &VC_FilteredSocketCount()
   &VC_RateQualityHandler(High Spawn Rate=60, Low Spawn Rate=30, Use Custom Rate?=true)
   &VC_SpawnRibbon_v1(Number of Ribbons=$Emitter.PortCount, SpawnRate=60)
  ]
  particleSpawnScript [
   &VC_PaintedVariantsOverride() {enabled=false}
   &VC_InitializeRibbon_v1(
    Enable Skeletal Position=true
    Lifetime=&MergeHelper_FloatSwitch(0=0.6, 1=0.6, 2=0.3, 3=?, Index=$User.GameModeID, Num=::NewEnumerator1)
    Number of Bones Per Boneflow group=2
    Offset Amount:NiagaraFloat=&MergeHelper_FloatSwitch(0=-48.1064, 1=-48.1064, 2=-15.1142, 3=?, Index=$User.GameModeID, Num=::NewEnumerator1)
    Offset Axis:ENiagaraOrientationAxis=::XAxis
   )
   &AddVelocity(Velocity=&MergeHelper_VectorSwitch(0=<-1300 0 0>, 1=<-1300 0 0>, 2=<-1300 0 0>, 3=?, Index=$User.GameModeID, Num=::NewEnumerator1)) {module=AddVelocity001}
   /SetVariables(
    Particles.WidthTag02:NiagaraFloat=&MergeHelper_FloatSwitch(0=49, 1=49, 2=49, 3=?, Index=0, Num=::NewEnumerator1)
    Particles.WidthTagMain=43
    Particles.float:NiagaraFloat=&SelectFloatFromArray(Direct Array Index=&SpawnGroup(), Float Selection Array=DataInterface() {propertyValues {FloatData [-42,0]}})
    Particles.float001:NiagaraFloat=&SelectFloatFromArray(Direct Array Index=&SpawnGroup(), Float Selection Array=DataInterface() {propertyValues {FloatData [0,42]}})
   ) {isSetParameters=true, module=SetVariables_#3}
  ]
  particleUpdateScript [
   &ParticleState()
   &InheritSourceMovement(
    Applied Position Scale=&Multiply_VectorByFloat(
     Vector=<1 1 1>
     Float=&FloatFromCurve(FloatCurve=DataInterface() {propertyValues {Curve {defaultValue=0, keys [<at 0 1 RCIM_Cubic>,<at 1 0.8 RCIM_Cubic>]}}})
    )
    ApplyForce=false
    ApplyPosition=true
    ApplyPositionFalloff=false
   )
   &SolveForcesAndVelocity(Clamp Velocity=false, Limit Acceleration=false)
   &ScaleColor(
    Scale Alpha=&FloatFromCurve(FloatCurve=DataInterface() {propertyValues {Curve {defaultValue=1, keys [{tangentMode=RCTM_User, time=0.2, value=0}, {tangentMode=RCTM_User, time=1, value=1}]}}})
    ScaleA=true
    ScaleRGB=true
   ) {enabled=false}
   &VC_UpdateRibbon_v1(Enable Skeletal Position=true)
   &Color(Color=&ColorFromCurve(ColorCurve=DataInterface() {propertyValues {
       AlphaCurve {keys [{arriveTangent=22.6751, interpMode=RCIM_Cubic, time=0, value=1}, <at 0.7826 1>]}
       BlueCurve {keys [<at 0 1>,<at 0.7865 0.7083>]}
       GreenCurve {keys [<at 0 1>,<at 0.7865 0.1564>]}
       RedCurve {keys [<at 0 1>,<at 0.7865 0.0885>]}
      }}))
   /SetVariables(
    Particles.BGB:LinearColor=&MultiplyLinearColor(
     Linear Color=<rgba 0 0.0052 0.25 50>
     Scale Factor=&MakeVector4FromVectorAndFloat(
      "Vector (RGB)"=<1 1 1>
      "Float (Alpha)"=&FloatFromCurve(FloatCurve=DataInterface() {propertyValues {Curve {defaultValue=0.878, keys [{tangentMode=RCTM_User, time=0, value=1}, {tangentMode=RCTM_User, time=1, value=1}]}}})
     )
    )
    Particles.BubblesColor:LinearColor=&MultiplyLinearColor(
     Linear Color=<rgba 0 0.0013 0.0625 50>
     Scale Factor=&MakeVector4FromVectorAndFloat(
      "Vector (RGB)"=<1 1 1>
      "Float (Alpha)"=&FloatFromCurve(FloatCurve=DataInterface() {propertyValues {Curve {keys [<at 0 1>,<at 1 1>]}}})
     )
    )
    Particles.BubblesW:LinearColor=&ColorFromCurve(ColorCurve=DataInterface() {propertyValues {
       AlphaCurve {keys [<at 0.6214 0>,<at 1 1>]}
       BlueCurve {keys [<at 0.6866 1>,<at 0.8703 1>]}
       GreenCurve {keys [<at 0.6866 1>,<at 0.8703 0.2968>]}
       RedCurve {keys [<at 0.6866 1>,<at 0.8703 0>]}
      }})
   ) {isSetParameters=true, module=SetVariables_#4}
  ]
  rendererClasses [&NiagaraRibbonRendererProperties]
  renderers [
   {rendererClass=&NiagaraRibbonRendererProperties}
   {rendererClass=&NiagaraRibbonRendererProperties, rendererIndex=1}
   {rendererClass=&NiagaraRibbonRendererProperties, rendererIndex=2}
   {rendererClass=&NiagaraRibbonRendererProperties, rendererIndex=3}
  ]
 }
 NE_DoodleMainTrail003 {
  bEnabled=false
  emitterUpdateScript [
   &EmitterState()
   &VC_FilteredSocketCount()
   &VC_RateQualityHandler()
   &VC_SpawnRibbon_v1(
    Number of Ribbons=$Emitter.PortCount
    SpawnRate=&MakeCustomFloatFromBool(Bool=$User.bIsFrontendPreview, True Float=60, False Float=30)
   )
  ]
  particleSpawnScript [
   &VC_PaintedVariantsOverride() {enabled=false}
   &VC_InitializeRibbon_v1(
    Color=<rgba 1 1 1 100>
    Enable Skeletal Position=true
    Lifetime=&MergeHelper_FloatSwitch(0=0.7, 1=0.7, 2=0.7, 3=?, Index=0, Num=::NewEnumerator1)
    Number of Bones Per Boneflow group=2
    Offset Amount:NiagaraFloat=&MergeHelper_FloatSwitch(0=-23, 1=-10, 2=-10, 3=?, Index=0, Num=::NewEnumerator1)
    Offset Axis:ENiagaraOrientationAxis=::XAxis
   )
   &AddVelocity(Velocity=&MergeHelper_VectorSwitch(0=<-1500 0 0>, 1=<-1500 0 0>, 2=<-1000 0 0>, 3=?, Index=0, Num=::NewEnumerator1)) {module=AddVelocity001}
  ]
  particleUpdateScript [
   &ParticleState()
   &InheritSourceMovement(
    Applied Position Scale=&Multiply_VectorByFloat(
     Vector=<1 1 1>
     Float=&FloatFromCurve(FloatCurve=DataInterface() {propertyValues {Curve {defaultValue=0, keys [<at 0 1 RCIM_Cubic>,<at 1 0.8 RCIM_Cubic>]}}})
    )
    ApplyForce=false
    ApplyPosition=true
    ApplyPositionFalloff=false
   )
   &SolveForcesAndVelocity(Clamp Velocity=false, Limit Acceleration=false)
   /SetVariables(
    Particles.ColorTag02:LinearColor=<rgba 0.0003 0.0065 0.0423 60>
    Particles.WidthTag02:NiagaraFloat=&MergeHelper_FloatSwitch(0=52, 1=0, 2=0, 3=?, Index=0, Num=::NewEnumerator1)
    Particles.WidthTagMain=42
   ) {isSetParameters=true, module=SetVariables_#3}
   &VC_UpdateRibbon_v1(Detach Locked Particle=true, Enable Skeletal Position=true, Locked Particle Detach Time=5)
   &Color(Color=&ColorFromCurve(ColorCurve=DataInterface() {propertyValues {
       AlphaCurve {keys [<at 0 100 RCIM_Cubic>,<at 1 0 RCIM_Cubic>]}
       BlueCurve {keys [<at 0 1 RCIM_Cubic>,<at 0.8802 1 RCIM_Cubic>]}
       GreenCurve {keys [<at 0 1 RCIM_Cubic>,<at 0.8802 0.0375 RCIM_Cubic>]}
       RedCurve {keys [<at 0 1 RCIM_Cubic>,<at 0.8802 0 RCIM_Cubic>]}
      }}))
  ]
  rendererClasses [&NiagaraRibbonRendererProperties]
  renderers [
   {rendererClass=&NiagaraRibbonRendererProperties}
   {rendererClass=&NiagaraRibbonRendererProperties, rendererIndex=1}
  ]
 }
 NE_DoodleMainTrail004 {
  bEnabled=false
  emitterUpdateScript [
   &EmitterState()
   &VC_FilteredSocketCount()
   &VC_RateQualityHandler(High Spawn Rate=60, Low Spawn Rate=30, Use Custom Rate?=true)
   &VC_SpawnRibbon_v1(Number of Ribbons=$Emitter.PortCount, SpawnRate=60)
  ]
  particleSpawnScript [
   &VC_PaintedVariantsOverride() {enabled=false}
   &VC_InitializeRibbon_v1(
    Enable Skeletal Position=true
    Lifetime=&MergeHelper_FloatSwitch(0=0.6, 1=0.6, 2=0.3, 3=?, Index=$User.GameModeID, Num=::NewEnumerator1)
    Number of Bones Per Boneflow group=2
    Offset Amount:NiagaraFloat=&MergeHelper_FloatSwitch(0=-70, 1=-35, 2=-35, 3=?, Index=$User.GameModeID, Num=::NewEnumerator1)
    Offset Axis:ENiagaraOrientationAxis=::XAxis
   )
   &AddVelocity(Velocity=&MergeHelper_VectorSwitch(0=<-1300 0 0>, 1=<-1300 0 0>, 2=<-1300 0 0>, 3=?, Index=$User.GameModeID, Num=::NewEnumerator1)) {module=AddVelocity001}
   /SetVariables(
    Particles.ColorTag02:LinearColor=&MultiplyLinearColor(
     Linear Color=<rgba 0 0.0013 0.0625 10>
     Scale Factor=&MakeVector4FromVectorAndFloat("Vector (RGB)"=<1 1 1>, "Float (Alpha)"=1)
    )
    Particles.WidthTag02:NiagaraFloat=&MergeHelper_FloatSwitch(0=52, 1=0, 2=0, 3=?, Index=0, Num=::NewEnumerator1)
    Particles.WidthTagMain=42
    Particles.float:NiagaraFloat=&SelectFloatFromArray(Direct Array Index=&SpawnGroup(), Float Selection Array=DataInterface() {propertyValues {FloatData [-42,0]}})
    Particles.float001:NiagaraFloat=&SelectFloatFromArray(Direct Array Index=&SpawnGroup(), Float Selection Array=DataInterface() {propertyValues {FloatData [0,42]}})
   ) {isSetParameters=true, module=SetVariables_#3}
  ]
  particleUpdateScript [
   &ParticleState()
   &InheritSourceMovement(
    Applied Position Scale=&Multiply_VectorByFloat(
     Vector=<1 1 1>
     Float=&FloatFromCurve(FloatCurve=DataInterface() {propertyValues {Curve {defaultValue=0, keys [<at 0 1 RCIM_Cubic>,<at 1 0.8 RCIM_Cubic>]}}})
    )
    ApplyForce=false
    ApplyPosition=true
    ApplyPositionFalloff=false
   )
   &SolveForcesAndVelocity(Clamp Velocity=false, Limit Acceleration=false)
   &ScaleColor(
    Scale Alpha=&FloatFromCurve(FloatCurve=DataInterface() {propertyValues {Curve {defaultValue=1, keys [{tangentMode=RCTM_User, time=0.2, value=0}, {tangentMode=RCTM_User, time=1, value=1}]}}})
    ScaleA=true
    ScaleRGB=true
   )
   &VC_UpdateRibbon_v1(Enable Skeletal Position=true)
   &Color(Color=&ColorFromCurve(ColorCurve=DataInterface() {propertyValues {
       AlphaCurve {keys [<at 0.1221 0>,<at 0.5148 1 RCIM_Cubic>,<at 1 0 RCIM_Cubic>]}
       BlueCurve {keys [<at 0.6107 1>,<at 0.8802 1 RCIM_Cubic>]}
       GreenCurve {keys [<at 0.6107 1>,<at 0.8802 0.0375 RCIM_Cubic>]}
       RedCurve {keys [<at 0.6107 1>,<at 0.8802 0 RCIM_Cubic>]}
      }}))
   /SetVariables(
    Particles.BGB:LinearColor=&MultiplyLinearColor(
     Linear Color=<rgba 0 0.0013 0.0625 2>
     Scale Factor=&MakeVector4FromVectorAndFloat(
      "Vector (RGB)"=<1 1 1>
      "Float (Alpha)"=&FloatFromCurve(FloatCurve=DataInterface() {propertyValues {Curve {defaultValue=0.878, keys [{tangentMode=RCTM_User, time=-0.1048, value=0}, {tangentMode=RCTM_User, time=0.3627, value=1}]}}})
     )
    )
    Particles.BGW:LinearColor=&MultiplyLinearColor(
     Linear Color=<rgba 1 1 1 0.8>
     Scale Factor=&MakeVector4FromVectorAndFloat(
      "Vector (RGB)"=<1 1 1>
      "Float (Alpha)"=&FloatFromCurve(FloatCurve=DataInterface() {propertyValues {Curve {defaultValue=0, keys [{tangentMode=RCTM_User, time=0, value=1}, {tangentMode=RCTM_User, time=0.189, value=0}]}}})
     )
    )
   ) {isSetParameters=true, module=SetVariables_#4}
  ]
  rendererClasses [&NiagaraRibbonRendererProperties]
  renderers [
   {rendererClass=&NiagaraRibbonRendererProperties}
   {rendererClass=&NiagaraRibbonRendererProperties, rendererIndex=1}
   {rendererClass=&NiagaraRibbonRendererProperties, rendererIndex=2}
   {rendererClass=&NiagaraRibbonRendererProperties, rendererIndex=3}
  ]
 }
 NE_VC_Boost_Sprite_Parent {
  emitterUpdateScript [
   &EmitterState(Inactive Response=::NewEnumerator1, Life Cycle Mode=::NewEnumerator1, UseLoopDelay=false)
   &VC_FilteredSocketCount()
   &SpawnRate(SpawnRate=20, Use Spawn Probability=false)
  ]
  particleSpawnScript [
   &VC_PaintedVariantsOverride() {enabled=false}
   &InitializeParticle(
    Color=<rgba 0 0.0013 0.0625 100>
    Lifetime=&RandomRangeFloat(Minimum=0.5, Maximum=0.9)
    Sprite Size Mode=::NewEnumerator3
    Uniform Sprite Size=&MergeHelper_FloatSwitch(0=14, 1=14, 2=11, 3=?, Index=$User.GameModeID, Num=::NewEnumerator1)
    UsePositionOffset=false
    Write Material Random=true
   )
   &PartitionParticles(Number Of Partitions=$Emitter.PortCount)
   &SkeletalMeshLocation(
    Bone / Socket Index=$Particles.PartitionParticles.Partition
    Bone / Socket Sampling Mode=::NewEnumerator5
    MaskSpawnGroup=false
    Mesh Sampling Type=::NewEnumerator6
    Skeletal Mesh=$System.SkeletalMesh
   )
   &AddVelocity(
    Cone Angle=20
    Module.UseConeFalloff=false
    Module.UseVelDistribution=false
    Override Seed=false
    Velocity Mode=::NewEnumerator2
    Velocity Speed=&MaskFloatBySpawnGroup(
     Masked Float=&MergeHelper_FloatSwitch(0=-500, 1=-500, 2=-500, 3=?, Index=$User.GameModeID, Num=::NewEnumerator1)
     Passthrough Float=0
    )
   )
   &ShapeLocation(MaskSpawnGroup=false, Override Seed=false, Sphere Radius=25, UseOwnerScale=true)
   /SetVariables(
    Particles.CoreColor:LinearColor=&ColorFromCurve(
     ColorCurve=DataInterface() {propertyValues {
       AlphaCurve {keys [<at 0 1 RCIM_Constant>,<at 1 0 RCIM_Cubic>]}
       BlueCurve {keys [<at 0 1 RCIM_Constant>,<at 0.5 0 RCIM_Cubic>]}
       GreenCurve {keys [<at 0 1 RCIM_Constant>,<at 0.5 0 RCIM_Cubic>]}
       RedCurve {keys [<at 0 1 RCIM_Constant>,<at 0.5 0.4179 RCIM_Cubic>]}
      }}
     CurveIndex=&RandomRangeFloat(Minimum=0, Maximum=1)
    )
    Particles.Linear Color=<rgba 0 0.0013 0.0625 100>
    Particles.RenderTag:NiagaraInt32=&SpawnGroup()
   ) {isSetParameters=true}
  ]
  particleUpdateScript [
   &ParticleState()
   &ScaleSpriteSize(
    Uniform Curve Scale=&RandomRangeFloat(Minimum=1, Maximum=3)
    Uniform Curve Sprite Scale=DataInterface() {propertyValues {Curve {keys [{tangentMode=RCTM_User, time=0, value=0}, <at 0.2 1>, <at 0.8 1>, {tangentMode=RCTM_User, time=1, value=0}]}}}
   )
   &SolveForcesAndVelocity(Clamp Velocity=false, Limit Acceleration=false)
   &CameraOffset(Camera Offset Amount=-100)
  ]
  rendererClasses [&NiagaraSpriteRendererProperties]
  renderers [
   {rendererClass=&NiagaraSpriteRendererProperties}
   {rendererClass=&NiagaraSpriteRendererProperties, rendererIndex=1}
  ]
 }
}
summary {
 systemName=NS_BoostDoodle
 emitters [
  {bEnabled=false, emitterName=NE_BoostCone001, rendererClasses [&NiagaraSpriteRendererProperties]}
  {emitterName=NE_VC_Boost_Sprite_Parent, rendererClasses [&NiagaraSpriteRendererProperties]}
  {bEnabled=false, emitterName=NE_DoodleMainTrail003, rendererClasses [&NiagaraRibbonRendererProperties]}
  {bEnabled=false, emitterName=NE_DoodleMainTrail004, rendererClasses [&NiagaraRibbonRendererProperties]}
  {bEnabled=false, emitterName=NE_DoodleMainTrail, rendererClasses [&NiagaraRibbonRendererProperties]}
  {emitterName=NE_BoostCone, rendererClasses [&NiagaraSpriteRendererProperties]}
  {emitterName=NE_DoodleMainTrail001, rendererClasses [&NiagaraRibbonRendererProperties]}
  {bEnabled=false, emitterName=NE_BoostCone002, rendererClasses [&NiagaraSpriteRendererProperties]}
  {bEnabled=false, emitterName=NE_BoostCone003, rendererClasses [&NiagaraSpriteRendererProperties]}
  {bEnabled=false, emitterName=NE_DoodleMainTrail002, rendererClasses [&NiagaraRibbonRendererProperties]}
 ]
 userVariables [
  User.bIsFrontendPreview {type=NiagaraBool}
  User.bIsPainted {type=NiagaraBool}
  User.ColorVariantID {type=NiagaraInt32}
  User.DTCarOffset {type=Vector3f, description="", defaultValue {struct=Vector3f, value=<0 0 0>}}
  User.GameModeID {type=NiagaraInt32}
  User.PaintedColor {type=LinearColor, description="", defaultValue {struct=LinearColor, value=<rgba 1 1 1 1>}}
  User.PaintedScalar {type=NiagaraFloat, defaultValue=1, description=""}
  User.PortCount {type=NiagaraInt32, defaultValue=1}
 ]
}
